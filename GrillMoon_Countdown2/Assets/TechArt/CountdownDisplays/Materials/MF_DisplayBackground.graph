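# Material function that generates a grayscale scrolling scanline background for countdown displays.
# It offsets the Y coordinate (UV green channel by default) by Time times scroll speed, scales it by frequency, and feeds a Sine.
# Saturate clips the negative half of the wave to leave dark gaps, and a Lerp scales the bands from black to the maximum brightness input.
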MaterialFunction /GrillMoon_Countdown2/Assets/TechArt/CountdownDisplays/Materials/MF_DisplayBackground
graph {nodes [
  n1 = Constant() {r=0.1}
  n2 = FunctionInput(Preview=ComponentMask(None=TextureCoordinate()) {g=true}) {bUsePreviewValueAsDefault=true, inputName=Y Coord, inputType=FunctionInput_Scalar}
  n3 = Multiply(
   A=FunctionInput(Preview=Constant()) {bUsePreviewValueAsDefault=true, inputName=Scanline Scroll Speed, inputType=FunctionInput_Scalar, sortPriority=2}
   B=Time()
  )
  n4 = Constant() {r=10}
  n5 = Sine(None=Multiply(A=Add(A=@n2, B=@n3), B=FunctionInput(Preview=@n4) {inputName=Scanline Frequency, inputType=FunctionInput_Scalar, sortPriority=1}))
  FunctionOutput(None=LinearInterpolate(
    B=FunctionInput(Preview=@n1) {bUsePreviewValueAsDefault=true, inputName=Scanline Max Brightness, inputType=FunctionInput_Scalar, sortPriority=3}
    Alpha=Saturate(None=@n5)
   ) {constB=0.5}) {outputName=Background Grayscale}
 ]}
