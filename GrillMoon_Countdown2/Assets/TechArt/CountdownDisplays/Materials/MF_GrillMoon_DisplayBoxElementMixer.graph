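# Material function for countdown display materials that mixes a new element into a box element, confined to the box's inner mask.
# The new element's color and mask are multiplied by the inner mask, then combined with the box element through the DisplayElementMixer function, with the box element mask defaulting to one.
# It outputs combined color and mask, passes through the inner mask, and can optionally max the combined mask into the whole-box mask for non-opaque boxes.
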MaterialFunction /GrillMoon_Countdown2/Assets/TechArt/CountdownDisplays/Materials/MF_GrillMoon_DisplayBoxElementMixer
graph {nodes [
  n1 = FunctionInput(Preview=Constant() {r=1}) {
   bUsePreviewValueAsDefault=true
   description=Set to 1 if not specified
   inputName=Box Element Mask
   inputType=FunctionInput_Scalar
   sortPriority=1
  }
  n2 = FunctionInput() {inputName=Box Inner Mask, inputType=FunctionInput_Scalar, sortPriority=2}
  n3 = MaterialFunctionCall!(
   Base Element Color=FunctionInput() {inputName=Box Element Color}
   Base Element Mask=@n1
   New Element Color=Multiply(A=FunctionInput() {inputName=New Element Color, sortPriority=4}, B=@n2)
   New Element Mask=Multiply(A=@n2, B=FunctionInput() {inputName=New Element Mask, inputType=FunctionInput_Scalar, sortPriority=5})
  ) {materialFunction=&/MF_GrillMoon_DisplayElementMixer}
  FunctionOutput(None=@n3.Combined Color) {outputName=Combined Color}
  FunctionOutput(None=@n2) {outputName=Inner Box Mask, sortPriority=2}
  FunctionOutput(None=@n3.Combined Mask) {outputName=Combined Mask, sortPriority=1}
  n4 = FunctionInput() {inputName=Whole Box Mask, inputType=FunctionInput_Scalar, sortPriority=3}
  n5 = FunctionInput(Preview=StaticBool()) {
   bUsePreviewValueAsDefault=true
   description="(Default False) Necessary when whole box is not completely opaque"
   inputName=Merge Combined Mask in to Whole Box Mask
   inputType=FunctionInput_StaticBool
   sortPriority=6
  }
  FunctionOutput(None=StaticSwitch(True=Max(A=@n3.Combined Mask, B=@n4), False=@n4, Value=@n5)) {outputName=Whole Box Mask, sortPriority=3}
 ]}
