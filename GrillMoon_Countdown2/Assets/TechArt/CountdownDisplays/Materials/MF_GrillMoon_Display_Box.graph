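# Material function for GrillMoon countdown displays that draws an outlined box between two UV positions, outputting color, masks, fitted and square UVs, and aspect ratio.
# It remaps input UVs into centered box space, then derives outline and fill masks using aspect-corrected outline width.
# Four static-switched corner-builder calls, fed mirrored UVs, warp the centered X coordinate to shape individual corners.
# Background opacity affects only the whole-box mask, not the color.
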MaterialFunction /GrillMoon_Countdown2/Assets/TechArt/CountdownDisplays/Materials/MF_GrillMoon_Display_Box
graph {nodes [
  FunctionOutput(None=NamedRerouteUsage()) {outputName=Inner Box Mask, sortPriority=1}
  n1 = FunctionInput() {inputName=Outline Color, sortPriority=4, previewValue=<1 1 1 1>}
  n2 = NamedRerouteUsage()
  n3 = FunctionInput() {inputName=Background Color, sortPriority=5}
  n4 = NamedRerouteUsage()
  FunctionOutput(None=Add(A=Multiply(A=@n1, B=@n2), B=Multiply(A=@n3, B=@n4))) {outputName=Box Color}
  n5 = FunctionInput(Preview=Constant() {r=1}) {bUsePreviewValueAsDefault=true, inputName=Background Opacity, inputType=FunctionInput_Scalar, sortPriority=6, previewValue=<1 0 0 1>}
  n6 = Multiply(A=OneMinus(None=@n5), B=NamedRerouteUsage())
  FunctionOutput(None=Multiply(A=OneMinus(None=@n6), B=NamedRerouteUsage())) {outputName=Whole Box Mask, sortPriority=2}
  FunctionOutput(None=NamedRerouteUsage()) {outputName=Box Fit UV, sortPriority=3}
  FunctionOutput(None=NamedRerouteUsage()) {outputName=Box Aspect Ratio, sortPriority=5}
  n7 = NamedRerouteUsage()
  n8 = NamedRerouteUsage()
  FunctionOutput(None=MaterialFunctionCall!(0-1 Position=NamedRerouteUsage(), "(XY, YX) Ratios"=AppendVector(A=@n7, B=@n8), Reconstruct YX From XY) {materialFunction=&/MF_Display_PositionFromAspectRatio}.Scaled Position) {outputName=Box Square UV, sortPriority=4}
  n9 = FunctionInput(Preview=Constant2Vector() {g=1, r=1}) {inputName=Bottom Right Position, inputType=FunctionInput_Vector2, sortPriority=2}
  n10 = FunctionInput(Preview=Constant2Vector()) {inputName=Top Left Position, inputType=FunctionInput_Vector2, sortPriority=1}
  n11 = Subtract(A=@n9, B=@n10)
  n12 = ComponentMask(None=@n11) {r=true}
  n13 = ComponentMask(None=@n11) {g=true}
  n14 = Constant() {r=1}
  NamedRerouteDeclaration(None=If(A=@n12, B=@n13, "A > B"=@n14, "A == B"=@n14, "A < B"=Constant())) {name="X>Y", nodeColor=<rgba 1 0 0.0146 1>}
  NamedRerouteDeclaration(None=Divide(A=@n12, B=@n13)) {name=XYRatio, nodeColor=<rgba 1 0 0.2619 1>}
  n15 = Constant() {r=0.1}
  n16 = NamedRerouteUsage()
  n17 = NamedRerouteUsage()
  n18 = NamedRerouteUsage()
  n19 = NamedRerouteUsage()
  n20 = NamedRerouteUsage()
  n21 = NamedRerouteUsage()
  n22 = NamedRerouteUsage()
  n23 = Multiply(
   A=Multiply(
    A=FunctionInput(Preview=@n15) {inputName=Outline Width, inputType=FunctionInput_Scalar, sortPriority=3}
    B=LinearInterpolate(A=@n16, B=@n17, Alpha=@n18) {constA=1}
   )
   B=AppendVector(A=LinearInterpolate(A=@n19, Alpha=@n20) {constA=1}, B=LinearInterpolate(B=@n21, Alpha=@n22) {constA=1})
  )
  n24 = Clamp(None=@n23) {maxDefault=0.999}
  n25 = NamedRerouteUsage()
  n26 = Saturate(None=InverseLinearInterpolate(B=OneMinus(None=@n24), Value=Abs(None=@n25)))
  n27 = Sign(None=@n25)
  NamedRerouteDeclaration(None=ConstantBiasScale(None=Multiply(A=@n26, B=@n27))) {name=BoxUV 0-1, nodeColor=<rgba 1 0.4453 0 1>}
  n28 = Abs(None=NamedRerouteUsage())
  n29 = MaterialFunctionCall(Float2=Ceil(None=OneMinus(None=@n28))) {materialFunction=&BreakOutFloat2Components}
  n30 = Multiply(A=@n29.R, B=@n29.R)
  NamedRerouteDeclaration(None=@n30) {name=WholeBoxMask, nodeColor=<rgba 0 0.3437 1 1>}
  n31 = InverseLinearInterpolate(A=OneMinus(None=Clamp(None=@n23) {minDefault=0.001}), Value=@n28)
  n32 = MaterialFunctionCall(Float2=Ceil(None=Saturate(None=@n31))) {materialFunction=&BreakOutFloat2Components}
  NamedRerouteDeclaration(None=Multiply(A=@n30, B=Max(A=@n32.R, B=@n32.R))) {name=OutlineMask, nodeColor=<rgba 0 0.3438 1 1>}
  NamedRerouteDeclaration(None=Divide(A=@n13, B=@n12)) {name=YXRatio, nodeColor=<rgba 1 0 0.2619 1>}
  n33 = NamedRerouteUsage()
  NamedRerouteDeclaration(None=Multiply(A=NamedRerouteUsage(), B=OneMinus(None=@n33))) {name=Inner Box Mask, nodeColor=<rgba 0 0.3438 1 1>}
  n34 = InverseLinearInterpolate(A=@n10, B=@n9, Value=FunctionInput(Preview=TextureCoordinate()) {inputName=Texture Coordinate 0-1, inputType=FunctionInput_Vector2})
  NamedRerouteDeclaration(None=ConstantBiasScale(None=Saturate(None=@n34)) {bias=-0.5, scale=2}) {name=RawBoxCenteredUV, nodeColor=<rgba 1 0.4453 0 1>}
  n35 = NamedRerouteUsage()
  n36 = NamedRerouteUsage()
  n37 = ComponentMask(None=@n36) {r=true}
  n38 = ComponentMask(None=@n36) {g=true}
  n39 = FunctionInput(Preview=Constant() {r=0.5}) {bUsePreviewValueAsDefault=true, inputName=Corner Width, inputType=FunctionInput_Scalar, sortPriority=7}
  n40 = NamedRerouteUsage()
  n41 = NamedRerouteUsage()
  n42 = NamedRerouteUsage()
  n43 = MaterialFunctionCall!(
   CornerUV=AppendVector(A=Multiply(A=@n37) {constB=-1}, B=Multiply(A=@n38) {constB=-1})
   "Input Corner Width (Scalar)"=@n39
   YXRatio=@n40
   "X>Y"=@n41
   XYRatio=@n42
  ) {materialFunction=&/MF_GrillMoon_BoxCornerBuilder}
  n44 = StaticBool()
  n45 = MaterialFunctionCall(CornerUV=AppendVector(A=@n37, B=Multiply(A=@n38) {constB=-1}), "Input Corner Width (Scalar)"=@n39, YXRatio=@n40, "X>Y"=@n41, XYRatio=@n42) {materialFunction=&MF_GrillMoon_BoxCornerBuilder}
  n46 = Constant()
  n47 = FunctionInput(Preview=StaticBool()) {bUsePreviewValueAsDefault=true, inputName=Top Right Corner, inputType=FunctionInput_StaticBool, sortPriority=9}
  n48 = MaterialFunctionCall(CornerUV=AppendVector(A=Multiply(A=@n37) {constB=-1}, B=@n38), "Input Corner Width (Scalar)"=@n39, YXRatio=@n40, "X>Y"=@n41, XYRatio=@n42) {materialFunction=&MF_GrillMoon_BoxCornerBuilder}
  n49 = StaticSwitch(
   True=@n48.CornerMask
   False=Constant()
   Value=FunctionInput(Preview=StaticBool()) {bUsePreviewValueAsDefault=true, inputName=Bottom Left Corner, inputType=FunctionInput_StaticBool, sortPriority=10}
  )
  n50 = MaterialFunctionCall(CornerUV=AppendVector(A=@n37, B=@n38), "Input Corner Width (Scalar)"=@n39, YXRatio=@n40, "X>Y"=@n41, XYRatio=@n42) {materialFunction=&MF_GrillMoon_BoxCornerBuilder}
  n51 = StaticSwitch(
   True=@n50.CornerMask
   False=Constant()
   Value=FunctionInput(Preview=StaticBool()) {bUsePreviewValueAsDefault=true, inputName=Bottom Right Corner, inputType=FunctionInput_StaticBool, sortPriority=11}
  )
  n52 = Add(
   A=StaticSwitch(
    True=@n43.CornerMask
    False=Constant()
    Value=FunctionInput(Preview=@n44) {bUsePreviewValueAsDefault=true, inputName=Top Left Corner, inputType=FunctionInput_StaticBool, sortPriority=8}
   )
   B=Add(A=StaticSwitch(True=@n45.CornerMask, False=@n46, Value=@n47), B=Add(A=@n49, B=@n51))
  )
  n53 = Multiply(A=ComponentMask(None=@n35) {r=true}, B=OneMinus(None=@n52))
  n54 = Add(A=@n48.CornerX, B=@n50.CornerX)
  n55 = Multiply(A=Add(A=@n43.CornerX, B=Add(A=@n45.CornerX, B=@n54)), B=@n52)
  NamedRerouteDeclaration(None=AppendVector(A=Add(A=@n53, B=@n55), B=ComponentMask(None=@n35) {g=true})) {name=CornerBoxCenteredUV, nodeColor=<rgba 1 0.4453 0 1>}
 ]}
