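# Material function for the GrillMoon countdown displays that fits a texture into a rectangular region of UV space.
# It inverse-lerps the input UV between top-left and bottom-right corners, then saturates it to output a remapped 0-1 UV.
# A second output is a binary rectangle mask, built branchlessly by centering that UV, applying Abs, OneMinus and Ceil, then multiplying both axes.
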MaterialFunction /GrillMoon_Countdown2/Assets/TechArt/CountdownDisplays/Materials/MF_GrillMoon_TextureOnDisplayUV
graph {nodes [
  n1 = FunctionInput() {inputName=Bottom Right Corner Location, inputType=FunctionInput_Vector2, sortPriority=2, previewValue=<1 1 0 1>}
  n2 = TextureCoordinate()
  n3 = Saturate(None=InverseLinearInterpolate(
    A=FunctionInput() {inputName=Top Left Corner Location, inputType=FunctionInput_Vector2, sortPriority=1}
    B=@n1
    Value=FunctionInput(Preview=@n2) {inputName=UV In, inputType=FunctionInput_Vector2}
   ))
  FunctionOutput(None=@n3) {outputName=UV Out}
  n4 = ConstantBiasScale(None=@n3) {bias=-0.5, scale=2}
  n5 = Ceil(None=OneMinus(None=Abs(None=@n4)))
  FunctionOutput(None=Multiply(A=ComponentMask(None=@n5) {r=true}, B=ComponentMask(None=@n5) {g=true})) {outputName=UV Mask, sortPriority=1}
 ]}
