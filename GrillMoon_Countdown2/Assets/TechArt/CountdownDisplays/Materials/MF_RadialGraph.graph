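# Material function that draws a radial countdown graph: a background ring, an inner bar with a fading trail, and a two-digit number.
# UVs are offset by graph location (center or corner anchored) and converted to radial angle and distance, with hard-edged ring masks cut from the distance.
# Angle offset by Time and wrapped with Frac sweeps the reversible trail, while MF_ScalarToWholeNumbers renders the digits.
# Colored layers are summed into Graph Color, alongside a combined Graph Mask.
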MaterialFunction /GrillMoon_Countdown2/Assets/TechArt/CountdownDisplays/Materials/MF_RadialGraph
graph {nodes [
  n1 = Constant() {r=0.5}
  n2 = Constant() {r=0.8}
  n3 = OneMinus(None=Clamp(None=FunctionInput(Preview=@n2) {bUsePreviewValueAsDefault=true, inputName=Inner Bar Trail Width, inputType=FunctionInput_Scalar, sortPriority=8}) {minDefault=0.001})
  n4 = FunctionInput() {bUsePreviewValueAsDefault=true, inputName=Graph Location, inputType=FunctionInput_Vector2, sortPriority=1}
  n5 = FunctionInput(Preview=StaticBool()) {
   bUsePreviewValueAsDefault=true
   description=Positioned From Center by Default
   inputName=Position Graph From Corner
   inputType=FunctionInput_StaticBool
   sortPriority=2
  }
  n6 = StaticSwitch(True=Divide(A=NamedRerouteUsage()), False=Constant(), Value=@n5)
  n7 = Add(
   A=FunctionInput(Preview=TextureCoordinate()) {bUsePreviewValueAsDefault=true, inputName=Texture Coordinate 0-1, inputType=FunctionInput_Vector2}
   B=ConstantBiasScale(None=Add(A=@n4, B=@n6)) {bias=-0.5, scale=-1}
  )
  n8 = MaterialFunctionCall(Vector or UVs=ConstantBiasScale(None=@n7) {bias=-0.5, scale=2}) {materialFunction=&VectorToRadialValue}
  n9 = FunctionInput(Preview=StaticBool()) {bUsePreviewValueAsDefault=true, inputName=Swap Trail Direction, inputType=FunctionInput_StaticBool, sortPriority=10}
  n10 = StaticSwitch(True=OneMinus(None=@n8.Vector Converted to Angle), False=@n8.Vector Converted to Angle, Value=@n9)
  n11 = FunctionInput(Preview=Time()) {inputName=Time, inputType=FunctionInput_Scalar, sortPriority=4}
  n12 = Frac(None=StaticSwitch(True=Add(A=@n10, B=@n11), False=Subtract(A=@n10, B=@n11), Value=@n9))
  n13 = InverseLinearInterpolate(
   B=Clamp(None=FunctionInput(Preview=@n1) {bUsePreviewValueAsDefault=true, inputName=Inner Bar Trail Fade, inputType=FunctionInput_Scalar, sortPriority=9}) {minDefault=0.001}
   Value=Saturate(None=InverseLinearInterpolate(A=@n3, Value=@n12))
  )
  n14 = Multiply(
   A=Multiply(A=Saturate(None=@n13), B=NamedRerouteUsage())
   B=FunctionInput() {inputName=Inner Bar Color, sortPriority=11, previewValue=<1 1 1 1>}
  )
  n15 = Multiply(A=FunctionInput() {inputName=Background Color, sortPriority=6, previewValue=<0.1 0.1 0.1 1>}, B=NamedRerouteUsage())
  n16 = NamedRerouteUsage()
  n17 = FunctionInput() {inputName=Numbers Color, sortPriority=14, previewValue=<1 1 1 1>}
  FunctionOutput(None=Add(A=Add(A=@n14, B=@n15), B=Multiply(A=@n16, B=@n17))) {outputName=Graph Color}
  FunctionOutput(None=Max(A=NamedRerouteUsage(), B=NamedRerouteUsage())) {outputName=Graph Mask, sortPriority=1}
  n18 = Constant() {r=0.3}
  n19 = OneMinus(None=Clamp(None=FunctionInput(Preview=@n18) {bUsePreviewValueAsDefault=true, inputName=Background Bar Width, inputType=FunctionInput_Scalar, sortPriority=5}) {maxDefault=0.999, minDefault=0.001})
  n20 = Constant() {r=1}
  n21 = Divide(
   A=@n8.Linear Distance
   B=NamedRerouteDeclaration(None=FunctionInput(Preview=@n20) {bUsePreviewValueAsDefault=true, inputName=Graph Size, inputType=FunctionInput_Scalar, sortPriority=3}) {name=Graph Size, nodeColor=<rgba 0 1 0.4141 1>}
  )
  n22 = InverseLinearInterpolate(A=@n19, Value=@n21)
  n23 = ConstantBiasScale(None=@n22) {bias=-0.5, scale=2}
  n24 = FunctionInput(Preview=Constant() {r=0.75}) {bUsePreviewValueAsDefault=true, inputName=Inner Bar Width, inputType=FunctionInput_Scalar, sortPriority=7}
  n25 = Saturate(None=Abs(None=Divide(A=@n23, B=@n24)))
  NamedRerouteDeclaration(None=Ceil(None=OneMinus(None=@n25))) {name=Inner Bar Mask, nodeColor=<rgba 0.6172 0 1 1>}
  n26 = InverseLinearInterpolate(B=@n19, Value=OneMinus(None=@n21))
  n27 = Min(A=@n26, B=@n22)
  NamedRerouteDeclaration(None=Ceil(None=Saturate(None=@n27))) {name=Background Bar Mask, nodeColor=<rgba 0.6172 0 1 1>}
  n28 = Constant()
  n29 = AppendVector(A=Constant() {r=1}, B=Constant() {r=0.66})
  n30 = Multiply(
   A=FunctionInput(Preview=Constant() {r=0.6}) {bUsePreviewValueAsDefault=true, inputName=Numbers Size, inputType=FunctionInput_Scalar, sortPriority=13}
   B=NamedRerouteUsage()
  )
  NamedRerouteDeclaration(None=MaterialFunctionCall!(
    Number Texture
    "NumbersToDisplay(int)"=FunctionInput(Preview=@n28) {bUsePreviewValueAsDefault=true, inputName="Numbers Value (int)", inputType=FunctionInput_Scalar, sortPriority=12}
    UVPadding=Multiply(A=@n29, B=@n30)
    UVs=@n7
    "NumberofDigits(int)"=Constant() {r=2}
   ) {materialFunction=&/MF_ScalarToWholeNumbers}.NumbersGreyscale) {name=Numbers Mask, nodeColor=<rgba 0.6172 0 1 1>}
  NamedRerouteDeclaration(None=Ceil(None=Saturate(None=@n26))) {name=Circle Mask, nodeColor=<rgba 0.6172 0 1 1>}
  NamedRerouteUsage()
 ]}
