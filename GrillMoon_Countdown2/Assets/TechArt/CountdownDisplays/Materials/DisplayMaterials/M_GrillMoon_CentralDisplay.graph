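# M_GrillMoon_CentralDisplay is a procedural sci-fi screen material for the GrillMoon countdown's central display.
# Aspect-corrected box functions lay out an 11-digit countdown with DHMS labels, a lower panel of static bars and shapes, and a right panel.
# The right panel combines a time-driven DragonGate flipbook, broken up by stepped noise, with two animated radial graphs.
# Mixer functions layer these over a scanline background, tinted by three UI color parameters.
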Material /GrillMoon_Countdown2/Assets/TechArt/CountdownDisplays/Materials/DisplayMaterials/M_GrillMoon_CentralDisplay
graph {
 properties {bAutomaticallySetUsageInEditor=false, usedShadingModels=Default Lit}
 nodes [
  Constant2Vector() {g=0.5, r=0.3}
  Constant2Vector() {g=0.9, r=0.325}
  Constant2Vector() {g=0.5, r=0.35}
  Constant2Vector() {g=0.9, r=0.375}
  Constant2Vector() {g=0.5, r=0.4}
  Constant2Vector() {g=0.9, r=0.475}
  Constant2Vector() {g=0.5, r=0.4}
  Constant2Vector() {g=0.9, r=0.425}
  Constant2Vector() {g=0.5, r=0.45}
  Constant2Vector() {g=0.9, r=0.475}
  n1 = ComponentMask(None=NamedRerouteUsage()) {g=true}
  n2 = ScalarParameter() {defaultValue=32, parameterName=Scanline Frequency}
  n3 = NamedRerouteUsage()
  n4 = Constant3Vector() {constant=<rgba 0 0 0 1>}
  n5 = MaterialFunctionCall!(
   Texture Coordinate 0-1=NamedRerouteUsage()
   Top Left Position=MaterialFunctionCall!(0-1 Position=Constant2Vector() {g=-0.01, r=-0.01}, "(XY, YX) Ratios"=@n3, Reconstruct YX From XY) {materialFunction=/GrillMoon_Countdown2/Assets/TechArt/CountdownDisplays/Materials/MF_Display_PositionFromAspectRatio}.Scaled Position
   Bottom Right Position=MaterialFunctionCall(0-1 Position=Constant2Vector() {g=0.5, r=0.6}, "(XY, YX) Ratios"=@n3) {materialFunction=&MF_Display_PositionFromAspectRatio}.Scaled Position
   Outline Width=Constant() {r=0.05}
   Outline Color=@n4
   Background Color=@n4
   Background Opacity=Constant()
   Corner Width=Constant()
   Top Left Corner
   Top Right Corner
   Bottom Left Corner
   Bottom Right Corner
  ) {materialFunction=/GrillMoon_Countdown2/Assets/TechArt/CountdownDisplays/Materials/MF_GrillMoon_Display_Box}
  n6 = Constant2Vector() {r=0.1}
  n7 = AppendVector(A=@n5.Box Aspect Ratio, B=Divide(B=@n5.Box Aspect Ratio))
  n8 = Constant2Vector() {g=1, r=0.9}
  n9 = MaterialFunctionCall!(
   UVs=MaterialFunctionCall!(
    UV In=@n5.Box Square UV
    Top Left Corner Location=MaterialFunctionCall(0-1 Position=@n6, "(XY, YX) Ratios"=@n7) {materialFunction=&MF_Display_PositionFromAspectRatio}.Scaled Position
    Bottom Right Corner Location=MaterialFunctionCall(0-1 Position=@n8, "(XY, YX) Ratios"=@n7) {materialFunction=&MF_Display_PositionFromAspectRatio}.Scaled Position
   ) {materialFunction=/GrillMoon_Countdown2/Assets/TechArt/CountdownDisplays/Materials/MF_GrillMoon_TextureOnDisplayUV}.UV Out
   UVPadding
   "NumberofDigits(int)"=Constant() {r=11}
  ) {materialFunction=/Game/Packages/Fortress_SharedMaterials/Base_Material_Functions/BaseCharacterShaderFunctions/MF_CountdownTimer}
  n10 = Add(
   A=AppendVector(
    A=ScalarParameter() {group=Numbers, parameterName=u_DHMSPos}
    B=ScalarParameter() {defaultValue=-0.6, group=Numbers, parameterName=v_DHMSPos}
   )
   B=@n9.UVS
  )
  n11 = AppendVector(
   A=ScalarParameter() {defaultValue=1.0383, group=Numbers, parameterName=u_DHMSScale}
   B=ScalarParameter() {defaultValue=0.419, group=Numbers, parameterName=v_DHMSScale}
  )
  n12 = Multiply(
   A=ScalarParameter() {defaultValue=0.65, parameterName=DHMS_Brightness}
   B=TextureSampleParameter2D(UVs=MaterialFunctionCall(UVs=@n10, Texture Scale=@n11) {materialFunction=&ScaleUVsByCenter}.UVs) {parameterName=DHMS, samplerType=SAMPLERTYPE_Masks, texture=/Game/Athena/Playlists/Barrier/Textures/Countdown_DHMS}.R
  )
  n13 = MaterialFunctionCall!(
   Box Element Color=@n5.Box Color
   Box Element Mask=Constant()
   Box Inner Mask=@n5.Box Color
   Whole Box Mask=@n5.Box Color
   New Element Color=Multiply(A=Add(A=@n9.NumbersGreyscale, B=@n12), B=NamedRerouteUsage())
   New Element Mask=@n9.NumbersGreyscale
   Merge Combined Mask in to Whole Box Mask
  ) {materialFunction=/GrillMoon_Countdown2/Assets/TechArt/CountdownDisplays/Materials/MF_GrillMoon_DisplayBoxElementMixer}
  n14 = StaticBool() {value=true}
  n15 = MaterialFunctionCall(
   Texture Coordinate 0-1=@n5.Box Square UV
   Top Left Position=MaterialFunctionCall(0-1 Position=Constant2Vector() {g=0.1, r=0.1}, "(XY, YX) Ratios"=@n7) {materialFunction=&MF_Display_PositionFromAspectRatio}.Scaled Position
   Bottom Right Position=MaterialFunctionCall(0-1 Position=Constant2Vector() {g=0.25, r=0.25}, "(XY, YX) Ratios"=@n7) {materialFunction=&MF_Display_PositionFromAspectRatio}.Scaled Position
   Outline Width=Constant() {r=0.05}
   Outline Color=Multiply(A=NamedRerouteUsage()) {constB=0.2}
   Background Color=NamedRerouteUsage()
   Corner Width=Constant() {r=0.5}
   Top Left Corner=@n14
   Bottom Right Corner=@n14
  ) {materialFunction=&MF_GrillMoon_Display_Box}
  n16 = MaterialFunctionCall(
   Box Element Color=@n13.Combined Color
   Box Element Mask=@n13.Combined Color
   Box Inner Mask=@n13.Combined Color
   Whole Box Mask=@n13.Combined Color
   New Element Color=@n15.Box Color
   New Element Mask=@n15.Box Color
   Merge Combined Mask in to Whole Box Mask=StaticBool() {value=true}
  ) {materialFunction=&MF_GrillMoon_DisplayBoxElementMixer}
  n17 = NamedRerouteUsage()
  n18 = Constant3Vector()
  n19 = MaterialFunctionCall(
   Texture Coordinate 0-1=NamedRerouteUsage()
   Top Left Position=MaterialFunctionCall(0-1 Position=Constant2Vector() {g=0.5, r=-0.01}, "(XY, YX) Ratios"=@n17) {materialFunction=&MF_Display_PositionFromAspectRatio}.Scaled Position
   Bottom Right Position=MaterialFunctionCall(0-1 Position=Constant2Vector() {g=1.01, r=0.6}, "(XY, YX) Ratios"=@n17) {materialFunction=&MF_Display_PositionFromAspectRatio}.Scaled Position
   Outline Width=Constant() {r=0.05}
   Outline Color=@n18
   Background Color=@n18
   Background Opacity=Constant()
   Corner Width=Constant()
  ) {materialFunction=&MF_GrillMoon_Display_Box}
  n20 = Constant() {r=0.3}
  n21 = Constant() {r=0.5}
  n22 = NamedRerouteUsage()
  n23 = Constant() {r=0.02}
  n24 = Add(A=@n20, B=Multiply(A=@n23))
  n25 = Constant() {r=0.9}
  n26 = MaterialFunctionCall(
   Texture Coordinate 0-1=NamedRerouteUsage()
   Top Left Position=MaterialFunctionCall(0-1 Position=AppendVector(A=@n20, B=@n21), "(XY, YX) Ratios"=@n22) {materialFunction=&MF_Display_PositionFromAspectRatio}.Scaled Position
   Bottom Right Position=MaterialFunctionCall(0-1 Position=AppendVector(A=@n24, B=@n25), "(XY, YX) Ratios"=@n22) {materialFunction=&MF_Display_PositionFromAspectRatio}.Scaled Position
   Outline Width=Constant()
   Outline Color=NamedRerouteUsage()
   Background Color=NamedRerouteUsage()
  ) {materialFunction=&MF_GrillMoon_Display_Box}
  n27 = Constant() {r=0.6}
  n28 = ComponentMask(None=@n26.Box Fit UV) {g=true}
  n29 = Ceil(None=Saturate(None=InverseLinearInterpolate(A=@n27, Value=@n28)))
  n30 = MaterialFunctionCall(
   Box Element Color=@n26.Box Color
   Box Inner Mask=@n26.Box Color
   Whole Box Mask=@n26.Box Color
   New Element Color=Multiply(A=@n29, B=NamedRerouteUsage())
   New Element Mask=@n29
  ) {materialFunction=&MF_GrillMoon_DisplayBoxElementMixer}
  n31 = MaterialFunctionCall(
   Box Element Color=@n19.Box Color
   Box Element Mask=Constant()
   Box Inner Mask=@n19.Box Color
   Whole Box Mask=@n19.Box Color
   New Element Color=@n30.Combined Color
   New Element Mask=@n30.Combined Color
  ) {materialFunction=&MF_GrillMoon_DisplayBoxElementMixer}
  n32 = Add(A=@n20, B=Multiply(A=@n23) {constB=2})
  n33 = NamedRerouteUsage()
  n34 = Add(A=@n20, B=Multiply(A=@n23) {constB=3})
  n35 = MaterialFunctionCall(
   Texture Coordinate 0-1=NamedRerouteUsage()
   Top Left Position=MaterialFunctionCall(0-1 Position=AppendVector(A=@n32, B=@n21), "(XY, YX) Ratios"=@n33) {materialFunction=&MF_Display_PositionFromAspectRatio}.Scaled Position
   Bottom Right Position=MaterialFunctionCall(0-1 Position=AppendVector(A=@n34, B=@n25), "(XY, YX) Ratios"=@n33) {materialFunction=&MF_Display_PositionFromAspectRatio}.Scaled Position
   Outline Width=Constant()
   Outline Color=NamedRerouteUsage()
   Background Color=NamedRerouteUsage()
  ) {materialFunction=&MF_GrillMoon_Display_Box}
  n36 = Constant() {r=0.9}
  n37 = ComponentMask(None=@n35.Box Fit UV) {g=true}
  n38 = Ceil(None=Saturate(None=InverseLinearInterpolate(A=@n36, Value=@n37)))
  n39 = MaterialFunctionCall(
   Box Element Color=@n35.Box Color
   Box Inner Mask=@n35.Box Color
   Whole Box Mask=@n35.Box Color
   New Element Color=Multiply(A=@n38, B=NamedRerouteUsage())
   New Element Mask=@n38
  ) {materialFunction=&MF_GrillMoon_DisplayBoxElementMixer}
  n40 = MaterialFunctionCall(
   Box Element Color=@n31.Combined Color
   Box Element Mask=@n31.Combined Color
   Box Inner Mask=@n31.Combined Color
   Whole Box Mask=@n31.Combined Color
   New Element Color=@n39.Combined Color
   New Element Mask=@n39.Combined Color
  ) {materialFunction=&MF_GrillMoon_DisplayBoxElementMixer}
  n41 = Add(A=@n20, B=Multiply(A=@n23) {constB=4})
  n42 = NamedRerouteUsage()
  n43 = Add(A=@n20, B=Multiply(A=@n23) {constB=5})
  n44 = MaterialFunctionCall(
   Texture Coordinate 0-1=NamedRerouteUsage()
   Top Left Position=MaterialFunctionCall(0-1 Position=AppendVector(A=@n41, B=@n21), "(XY, YX) Ratios"=@n42) {materialFunction=&MF_Display_PositionFromAspectRatio}.Scaled Position
   Bottom Right Position=MaterialFunctionCall(0-1 Position=AppendVector(A=@n43, B=@n25), "(XY, YX) Ratios"=@n42) {materialFunction=&MF_Display_PositionFromAspectRatio}.Scaled Position
   Outline Width=Constant()
   Outline Color=NamedRerouteUsage()
   Background Color=NamedRerouteUsage()
  ) {materialFunction=&MF_GrillMoon_Display_Box}
  n45 = Constant() {r=0.4}
  n46 = ComponentMask(None=@n44.Box Fit UV) {g=true}
  n47 = Ceil(None=Saturate(None=InverseLinearInterpolate(A=@n45, Value=@n46)))
  n48 = MaterialFunctionCall(
   Box Element Color=@n44.Box Color
   Box Inner Mask=@n44.Box Color
   Whole Box Mask=@n44.Box Color
   New Element Color=Multiply(A=@n47, B=NamedRerouteUsage())
   New Element Mask=@n47
  ) {materialFunction=&MF_GrillMoon_DisplayBoxElementMixer}
  n49 = MaterialFunctionCall(
   Box Element Color=@n40.Combined Color
   Box Element Mask=@n40.Combined Color
   Box Inner Mask=@n40.Combined Color
   Whole Box Mask=@n40.Combined Color
   New Element Color=@n48.Combined Color
   New Element Mask=@n48.Combined Color
  ) {materialFunction=&MF_GrillMoon_DisplayBoxElementMixer}
  n50 = Add(A=@n20, B=Multiply(A=@n23) {constB=6})
  n51 = NamedRerouteUsage()
  n52 = Add(A=@n20, B=Multiply(A=@n23) {constB=7})
  n53 = MaterialFunctionCall(
   Texture Coordinate 0-1=NamedRerouteUsage()
   Top Left Position=MaterialFunctionCall(0-1 Position=AppendVector(A=@n50, B=@n21), "(XY, YX) Ratios"=@n51) {materialFunction=&MF_Display_PositionFromAspectRatio}.Scaled Position
   Bottom Right Position=MaterialFunctionCall(0-1 Position=AppendVector(A=@n52, B=@n25), "(XY, YX) Ratios"=@n51) {materialFunction=&MF_Display_PositionFromAspectRatio}.Scaled Position
   Outline Width=Constant()
   Outline Color=NamedRerouteUsage()
   Background Color=NamedRerouteUsage()
  ) {materialFunction=&MF_GrillMoon_Display_Box}
  n54 = Constant() {r=0.3}
  n55 = ComponentMask(None=@n53.Box Fit UV) {g=true}
  n56 = Ceil(None=Saturate(None=InverseLinearInterpolate(A=@n54, Value=@n55)))
  n57 = MaterialFunctionCall(
   Box Element Color=@n53.Box Color
   Box Inner Mask=@n53.Box Color
   Whole Box Mask=@n53.Box Color
   New Element Color=Multiply(A=@n56, B=NamedRerouteUsage())
   New Element Mask=@n56
  ) {materialFunction=&MF_GrillMoon_DisplayBoxElementMixer}
  n58 = MaterialFunctionCall(
   Box Element Color=@n49.Combined Color
   Box Element Mask=@n49.Combined Color
   Box Inner Mask=@n49.Combined Color
   Whole Box Mask=@n49.Combined Color
   New Element Color=@n57.Combined Color
   New Element Mask=@n57.Combined Color
  ) {materialFunction=&MF_GrillMoon_DisplayBoxElementMixer}
  n59 = Add(A=@n20, B=Multiply(A=@n23) {constB=8})
  n60 = NamedRerouteUsage()
  n61 = Add(A=@n20, B=Multiply(A=@n23) {constB=9})
  n62 = MaterialFunctionCall(
   Texture Coordinate 0-1=NamedRerouteUsage()
   Top Left Position=MaterialFunctionCall(0-1 Position=AppendVector(A=@n59, B=@n21), "(XY, YX) Ratios"=@n60) {materialFunction=&MF_Display_PositionFromAspectRatio}.Scaled Position
   Bottom Right Position=MaterialFunctionCall(0-1 Position=AppendVector(A=@n61, B=@n25), "(XY, YX) Ratios"=@n60) {materialFunction=&MF_Display_PositionFromAspectRatio}.Scaled Position
   Outline Width=Constant()
   Outline Color=NamedRerouteUsage()
   Background Color=NamedRerouteUsage()
  ) {materialFunction=&MF_GrillMoon_Display_Box}
  n63 = Constant() {r=0.7}
  n64 = ComponentMask(None=@n62.Box Fit UV) {g=true}
  n65 = Ceil(None=Saturate(None=InverseLinearInterpolate(A=@n63, Value=@n64)))
  n66 = MaterialFunctionCall(
   Box Element Color=@n62.Box Color
   Box Inner Mask=@n62.Box Color
   Whole Box Mask=@n62.Box Color
   New Element Color=Multiply(A=@n65, B=NamedRerouteUsage())
   New Element Mask=@n65
  ) {materialFunction=&MF_GrillMoon_DisplayBoxElementMixer}
  n67 = MaterialFunctionCall(
   Box Element Color=@n58.Combined Color
   Box Element Mask=@n58.Combined Color
   Box Inner Mask=@n58.Combined Color
   Whole Box Mask=@n58.Combined Color
   New Element Color=@n66.Combined Color
   New Element Mask=@n66.Combined Color
  ) {materialFunction=&MF_GrillMoon_DisplayBoxElementMixer}
  n68 = NamedRerouteUsage()
  n69 = StaticBool() {value=true}
  n70 = MaterialFunctionCall(
   Texture Coordinate 0-1=NamedRerouteUsage()
   Top Left Position=MaterialFunctionCall(0-1 Position=Constant2Vector() {g=0.8, r=0.05}, "(XY, YX) Ratios"=@n68) {materialFunction=&MF_Display_PositionFromAspectRatio}.Scaled Position
   Bottom Right Position=MaterialFunctionCall(0-1 Position=Constant2Vector() {g=0.9, r=0.25}, "(XY, YX) Ratios"=@n68) {materialFunction=&MF_Display_PositionFromAspectRatio}.Scaled Position
   Outline Width=Constant() {r=0.05}
   Outline Color=Multiply(A=NamedRerouteUsage()) {constB=0.2}
   Background Color=NamedRerouteUsage()
   Corner Width=Constant() {r=0.5}
   Top Left Corner=@n69
   Bottom Right Corner=@n69
  ) {materialFunction=&MF_GrillMoon_Display_Box}
  n71 = MaterialFunctionCall(
   Box Element Color=@n67.Combined Color
   Box Element Mask=@n67.Combined Color
   Box Inner Mask=@n67.Combined Color
   Whole Box Mask=@n67.Combined Color
   New Element Color=@n70.Box Color
   New Element Mask=@n70.Box Color
  ) {materialFunction=&MF_GrillMoon_DisplayBoxElementMixer}
  n72 = NamedRerouteUsage()
  n73 = StaticBool() {value=true}
  n74 = MaterialFunctionCall(
   Texture Coordinate 0-1=NamedRerouteUsage()
   Top Left Position=MaterialFunctionCall(0-1 Position=Constant2Vector() {g=0.65, r=0.05}, "(XY, YX) Ratios"=@n72) {materialFunction=&MF_Display_PositionFromAspectRatio}.Scaled Position
   Bottom Right Position=MaterialFunctionCall(0-1 Position=Constant2Vector() {g=0.75, r=0.25}, "(XY, YX) Ratios"=@n72) {materialFunction=&MF_Display_PositionFromAspectRatio}.Scaled Position
   Outline Width=Constant() {r=0.05}
   Outline Color=Multiply(A=NamedRerouteUsage()) {constB=0.2}
   Background Color=NamedRerouteUsage()
   Corner Width=Constant() {r=0.5}
   Top Left Corner=@n73
   Bottom Right Corner=@n73
  ) {materialFunction=&MF_GrillMoon_Display_Box}
  n75 = MaterialFunctionCall(
   Box Element Color=@n71.Combined Color
   Box Element Mask=@n71.Combined Color
   Box Inner Mask=@n71.Combined Color
   Whole Box Mask=@n71.Combined Color
   New Element Color=@n74.Box Color
   New Element Mask=@n74.Box Color
  ) {materialFunction=&MF_GrillMoon_DisplayBoxElementMixer}
  n76 = NamedRerouteUsage()
  n77 = StaticBool() {value=true}
  n78 = MaterialFunctionCall(
   Texture Coordinate 0-1=NamedRerouteUsage()
   Top Left Position=MaterialFunctionCall(0-1 Position=Constant2Vector() {g=0.5, r=0.05}, "(XY, YX) Ratios"=@n76) {materialFunction=&MF_Display_PositionFromAspectRatio}.Scaled Position
   Bottom Right Position=MaterialFunctionCall(0-1 Position=Constant2Vector() {g=0.6, r=0.25}, "(XY, YX) Ratios"=@n76) {materialFunction=&MF_Display_PositionFromAspectRatio}.Scaled Position
   Outline Width=Constant() {r=0.05}
   Outline Color=Multiply(A=NamedRerouteUsage()) {constB=0.2}
   Background Color=NamedRerouteUsage()
   Corner Width=Constant() {r=0.5}
   Top Left Corner=@n77
   Bottom Right Corner=@n77
  ) {materialFunction=&MF_GrillMoon_Display_Box}
  n79 = MaterialFunctionCall(
   Box Element Color=@n75.Combined Color
   Box Element Mask=@n75.Combined Color
   Box Inner Mask=@n75.Combined Color
   Whole Box Mask=@n75.Combined Color
   New Element Color=@n78.Box Color
   New Element Mask=@n78.Box Color
  ) {materialFunction=&MF_GrillMoon_DisplayBoxElementMixer}
  n80 = NamedRerouteUsage()
  n81 = StaticBool() {value=true}
  n82 = MaterialFunctionCall(
   Texture Coordinate 0-1=NamedRerouteUsage()
   Top Left Position=MaterialFunctionCall(0-1 Position=Constant2Vector() {g=0.5, r=0.5}, "(XY, YX) Ratios"=@n80) {materialFunction=&MF_Display_PositionFromAspectRatio}.Scaled Position
   Bottom Right Position=MaterialFunctionCall(0-1 Position=Constant2Vector() {g=0.9, r=0.9}, "(XY, YX) Ratios"=@n80) {materialFunction=&MF_Display_PositionFromAspectRatio}.Scaled Position
   Outline Width=Constant() {r=0.1}
   Outline Color=NamedRerouteUsage()
   Background Color=NamedRerouteUsage()
   Corner Width=Constant() {r=0.5}
   Top Left Corner=@n81
   Bottom Right Corner=@n81
  ) {materialFunction=&MF_GrillMoon_Display_Box}
  n83 = MaterialFunctionCall(
   Box Element Color=@n79.Combined Color
   Box Element Mask=@n79.Combined Color
   Box Inner Mask=@n79.Combined Color
   Whole Box Mask=@n79.Combined Color
   New Element Color=@n82.Box Color
   New Element Mask=@n82.Box Color
  ) {materialFunction=&MF_GrillMoon_DisplayBoxElementMixer}
  n84 = NamedRerouteUsage()
  n85 = MaterialFunctionCall(
   Texture Coordinate 0-1=NamedRerouteUsage()
   Top Left Position=MaterialFunctionCall(0-1 Position=Constant2Vector() {g=0.2, r=0.1}, "(XY, YX) Ratios"=@n84) {materialFunction=&MF_Display_PositionFromAspectRatio}.Scaled Position
   Bottom Right Position=MaterialFunctionCall(0-1 Position=Constant2Vector() {g=0.3, r=0.9}, "(XY, YX) Ratios"=@n84) {materialFunction=&MF_Display_PositionFromAspectRatio}.Scaled Position
   Outline Width=Constant() {r=0.1}
   Outline Color=NamedRerouteUsage()
   Background Color=Constant3Vector()
   Background Opacity=Constant()
  ) {materialFunction=&MF_GrillMoon_Display_Box}
  n86 = Constant() {r=0.6}
  n87 = Constant() {r=0.9}
  n88 = MaterialFunctionCall(
   UV In=@n85.Box Square UV
   Top Left Corner Location=Constant2Vector() {g=0.1, r=0.1}
   Bottom Right Corner Location=MaterialFunctionCall(0-1 Position=AppendVector(A=@n86, B=@n87), "(XY, YX) Ratios"=@n85.Box Aspect Ratio, Reconstruct YX From XY=StaticBool() {value=true}) {materialFunction=&MF_Display_PositionFromAspectRatio}.Scaled Position
  ) {materialFunction=&MF_GrillMoon_TextureOnDisplayUV}
  n89 = Ceil(None=ComponentMask(None=@n88.UV Out) {g=true})
  n90 = MaterialFunctionCall(
   Box Element Color=@n85.Box Color
   Box Element Mask=Constant()
   Box Inner Mask=@n85.Box Color
   Whole Box Mask=@n85.Box Color
   New Element Color=Multiply(A=Multiply(A=@n89, B=@n88.UV Mask), B=NamedRerouteUsage())
   New Element Mask=@n88.UV Mask
   Merge Combined Mask in to Whole Box Mask=StaticBool() {value=true}
  ) {materialFunction=&MF_GrillMoon_DisplayBoxElementMixer}
  n91 = MaterialFunctionCall(
   Box Element Color=@n83.Combined Color
   Box Element Mask=@n83.Combined Color
   Box Inner Mask=@n83.Combined Color
   Whole Box Mask=@n83.Combined Color
   New Element Color=@n90.Combined Color
   New Element Mask=@n90.Combined Color
   Merge Combined Mask in to Whole Box Mask=StaticBool() {value=true}
  ) {materialFunction=&MF_GrillMoon_DisplayBoxElementMixer}
  n92 = MaterialFunctionCall!(
   Base Element Color=@n16.Combined Color
   Base Element Mask=@n16.Combined Color
   New Element Color=@n91.Combined Color
   New Element Mask=@n91.Combined Color
  ) {materialFunction=/GrillMoon_Countdown2/Assets/TechArt/CountdownDisplays/Materials/MF_GrillMoon_DisplayElementMixer}
  n93 = NamedRerouteUsage()
  n94 = Constant3Vector()
  n95 = MaterialFunctionCall(
   Texture Coordinate 0-1=NamedRerouteUsage()
   Top Left Position=MaterialFunctionCall(0-1 Position=Constant2Vector() {g=-0.02, r=0.6}, "(XY, YX) Ratios"=@n93) {materialFunction=&MF_Display_PositionFromAspectRatio}.Scaled Position
   Bottom Right Position=MaterialFunctionCall(0-1 Position=Constant2Vector() {g=1.02, r=1.01}, "(XY, YX) Ratios"=@n93) {materialFunction=&MF_Display_PositionFromAspectRatio}.Scaled Position
   Outline Width=Constant() {r=0.05}
   Outline Color=@n94
   Background Color=@n94
   Background Opacity=Constant()
   Corner Width=Constant()
  ) {materialFunction=&MF_GrillMoon_Display_Box}
  n96 = MaterialFunctionCall(
   UV In=@n95.Box Square UV
   Top Left Corner Location=Constant2Vector() {g=0.3, r=0.05}
   Bottom Right Corner Location=Constant2Vector() {g=1.2, r=0.95}
  ) {materialFunction=&MF_GrillMoon_TextureOnDisplayUV}
  n97 = Modulo!(A=Multiply(A=Time(), B=ScalarParameter() {defaultValue=0.1, parameterName=Wireframe Rotation Speed}), B=Constant() {r=1})
  n98 = Constant() {r=2}
  n99 = Constant() {r=4}
  n100 = Multiply(A=@n97, B=Multiply(A=@n98, B=@n99))
  n101 = TextureSample(UVs=Panner(Coordinate=Multiply(A=@n96.UV Out) {constB=3}, Time=Floor(None=@n100)) {speedX=0.1}) {samplerType=SAMPLERTYPE_Grayscale, texture=&LowResBlurredNoise}
  n102 = MaterialFunctionCall(
   "Animation  Phase (0-1)"=@n97
   Number of Rows=@n98
   Number of Columns=@n99
   Texture=TextureObject() {texture=/GrillMoon_Countdown2/Assets/TechArt/CountdownDisplays/Textures/DragonGate_SpriteSheet_2048_RG}
   UVs=@n96.UV Out
  ) {materialFunction=&FlipBook}
  n103 = ComponentMask(None=@n102.Result) {g=true}
  n104 = NamedRerouteUsage()
  n105 = Multiply(A=ComponentMask(None=@n102.Result) {r=true}, B=@n103)
  n106 = MaterialFunctionCall(
   Base Element Color=Multiply(A=Multiply(A=@n101.R, B=@n103), B=@n104)
   Base Element Mask=@n103
   New Element Color=Multiply(A=@n105, B=@n104)
   New Element Mask=@n105
  ) {materialFunction=&MF_GrillMoon_DisplayElementMixer}
  n107 = MaterialFunctionCall(
   Box Element Color=@n95.Box Color
   Box Element Mask=Constant()
   Box Inner Mask=@n95.Box Color
   Whole Box Mask=@n95.Box Color
   New Element Color=@n106.Combined Color
   New Element Mask=@n106.Combined Color
  ) {materialFunction=&MF_GrillMoon_DisplayBoxElementMixer}
  n108 = AppendVector(A=@n95.Box Aspect Ratio, B=Divide(B=@n95.Box Aspect Ratio))
  n109 = NamedRerouteUsage()
  n110 = MaterialFunctionCall!(
   Texture Coordinate 0-1=@n95.Box Square UV
   Graph Location=MaterialFunctionCall(0-1 Position=Constant2Vector() {g=0.2, r=0.55}, "(XY, YX) Ratios"=@n108) {materialFunction=&MF_Display_PositionFromAspectRatio}.Scaled Position
   Position Graph From Corner
   Graph Size=Constant() {r=0.2}
   Time=Time()
   Background Bar Width
   Background Color=NamedRerouteUsage()
   Inner Bar Width
   Inner Bar Trail Width
   Inner Bar Trail Fade
   Swap Trail Direction
   Inner Bar Color=@n109
   "Numbers Value (int)"=Constant() {r=42}
   Numbers Size
   Numbers Color=@n109
  ) {materialFunction=/GrillMoon_Countdown2/Assets/TechArt/CountdownDisplays/Materials/MF_RadialGraph}
  n111 = MaterialFunctionCall(
   Box Element Color=@n107.Combined Color
   Box Element Mask=@n107.Combined Color
   Box Inner Mask=@n107.Combined Color
   Whole Box Mask=@n107.Combined Color
   New Element Color=@n110.Graph Color
   New Element Mask=@n110.Graph Color
  ) {materialFunction=&MF_GrillMoon_DisplayBoxElementMixer}
  n112 = NamedRerouteUsage()
  n113 = MaterialFunctionCall(
   Texture Coordinate 0-1=@n95.Box Square UV
   Graph Location=MaterialFunctionCall(0-1 Position=Constant2Vector() {g=0.2, r=0.8}, "(XY, YX) Ratios"=@n108) {materialFunction=&MF_Display_PositionFromAspectRatio}.Scaled Position
   Graph Size=Constant() {r=0.2}
   Time=Time()
   Background Color=NamedRerouteUsage()
   Inner Bar Color=@n112
   "Numbers Value (int)"=Constant() {r=23}
   Numbers Color=@n112
  ) {materialFunction=&MF_RadialGraph}
  n114 = MaterialFunctionCall(
   Box Element Color=@n111.Combined Color
   Box Element Mask=@n111.Combined Color
   Box Inner Mask=@n111.Combined Color
   Whole Box Mask=@n111.Combined Color
   New Element Color=@n113.Graph Color
   New Element Mask=@n113.Graph Color
   Merge Combined Mask in to Whole Box Mask=StaticBool() {value=true}
  ) {materialFunction=&MF_GrillMoon_DisplayBoxElementMixer}
  n115 = MaterialFunctionCall(
   Base Element Color=@n92.Combined Color
   Base Element Mask=@n92.Combined Color
   New Element Color=@n114.Combined Color
   New Element Mask=@n114.Combined Color
  ) {materialFunction=&MF_GrillMoon_DisplayElementMixer}
  MaterialFunctionCall(
   Base Element Color=LinearInterpolate(
    A=NamedRerouteUsage()
    B=NamedRerouteUsage()
    Alpha=MaterialFunctionCall!(Y Coord=@n1, Scanline Frequency=@n2, Scanline Scroll Speed, Scanline Max Brightness) {materialFunction=/GrillMoon_Countdown2/Assets/TechArt/CountdownDisplays/Materials/MF_DisplayBackground}.Background Grayscale
   )
   New Element Color=@n115.Combined Color
   New Element Mask=@n115.Combined Color
  ) {materialFunction=&MF_GrillMoon_DisplayElementMixer}
  NamedRerouteDeclaration(None=ComponentMask(None=VectorParameter() {parameterName=UI Secondary Color, defaultValue=<rgba 0.0003 0.013 0.0252 1>}.RGB) {b=true, g=true, r=true}) {name=UI Secondary Color, nodeColor=<rgba 0 1 0.1562 1>}
  NamedRerouteDeclaration(None=ComponentMask(None=VectorParameter() {parameterName=UI Primary Color, defaultValue=<rgba 0.3467 1 0.9823 1>}.RGB) {b=true, g=true, r=true}) {name=UI Primary Color, nodeColor=<rgba 0 1 0.1562 1>}
  NamedRerouteDeclaration(None=ComponentMask(None=VectorParameter() {parameterName=UI Tertiary Color, defaultValue=<rgba 0 0.7347 1 1>}.RGB) {b=true, g=true, r=true}) {name=UI Tertiary Color, nodeColor=<rgba 0 1 0.1562 1>}
  NamedRerouteDeclaration(None=AppendVector(A=@n19.Box Aspect Ratio, B=Divide(B=@n19.Box Aspect Ratio))) {name="Bottom Display (XY, YX)", nodeColor=<rgba 1 0.1172 0 1>}
  NamedRerouteDeclaration(None=MaterialFunctionCall(0-1 Position=TextureCoordinate(), "(XY, YX) Ratios"=NamedRerouteUsage()) {materialFunction=&MF_Display_PositionFromAspectRatio}.Scaled Position) {name=Aspect Correct UV, nodeColor=<rgba 1 0 0.2297 1>}
  n116 = ScalarParameter() {defaultValue=1.7778, parameterName=Screen Aspect Ratio}
  NamedRerouteDeclaration(None=AppendVector(A=@n116, B=Divide(B=@n116))) {name="(XY, YX) Ratios", nodeColor=<rgba 0 0.5781 1 1>}
  NamedRerouteDeclaration(None=@n19.Box Square UV) {name=Bottom Display UV, nodeColor=<rgba 1 0.1172 0 1>}
 ]
}
