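# Material for a GrillMoon countdown top display, drawing an animated sci-fi style UI screen procedurally from shared display material functions.
# A line-graph box, a smaller box and a numbered radial graph are layered through element mixers over a scanline background.
# The line graph pans in discrete steps using floored time.
# Named reroutes share UI color parameters and aspect-ratio-corrected UVs, with element positions scaled by the screen aspect ratio parameter.
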Material /GrillMoon_Countdown2/Assets/TechArt/CountdownDisplays/Materials/DisplayMaterials/M_GrillMoon_TopDisplay_0
graph {
 properties {bAutomaticallySetUsageInEditor=false, usedShadingModels=Default Lit}
 nodes [
  n1 = ComponentMask(None=NamedRerouteUsage()) {g=true}
  n2 = ScalarParameter() {defaultValue=32, parameterName=Scanline Frequency}
  n3 = NamedRerouteUsage()
  n4 = StaticBool() {value=true}
  n5 = MaterialFunctionCall!(
   Texture Coordinate 0-1=NamedRerouteUsage()
   Top Left Position=MaterialFunctionCall!(0-1 Position=Constant2Vector() {g=0.3, r=0.3}, "(XY, YX) Ratios"=@n3, Reconstruct YX From XY) {materialFunction=/GrillMoon_Countdown2/Assets/TechArt/CountdownDisplays/Materials/MF_Display_PositionFromAspectRatio}.Scaled Position
   Bottom Right Position=MaterialFunctionCall(0-1 Position=Constant2Vector() {g=0.9, r=0.9}, "(XY, YX) Ratios"=@n3) {materialFunction=&MF_Display_PositionFromAspectRatio}.Scaled Position
   Outline Width=Constant() {r=0.05}
   Outline Color=NamedRerouteUsage()
   Background Color=NamedRerouteUsage()
   Background Opacity
   Corner Width=Constant() {r=0.3}
   Top Left Corner=@n4
   Top Right Corner
   Bottom Left Corner
   Bottom Right Corner=@n4
  ) {materialFunction=/GrillMoon_Countdown2/Assets/TechArt/CountdownDisplays/Materials/MF_GrillMoon_Display_Box}
  n6 = Multiply(A=ComponentMask(None=@n5.Box Square UV) {r=true}) {constB=0.25}
  n7 = ComponentMask(None=@n5.Box Square UV) {g=true}
  n8 = Multiply(A=Time(), B=ScalarParameter() {defaultValue=1, parameterName=Graph Speed})
  n9 = TextureSample(UVs=Panner(Coordinate=AppendVector(A=@n6, B=@n7), Time=Floor(None=@n8)) {speedX=0.062}) {texture=/GrillMoon_Countdown2/Assets/TechArt/CountdownDisplays/Textures/T_LineGraph_RG}
  n10 = MaterialFunctionCall!(
   Box Element Color=@n5.Box Color
   Box Element Mask
   Box Inner Mask=@n5.Box Color
   Whole Box Mask=@n5.Box Color
   New Element Color=Multiply(A=@n9.G, B=NamedRerouteUsage())
   New Element Mask=@n9.G
   Merge Combined Mask in to Whole Box Mask
  ) {materialFunction=/GrillMoon_Countdown2/Assets/TechArt/CountdownDisplays/Materials/MF_GrillMoon_DisplayBoxElementMixer}
  n11 = NamedRerouteUsage()
  n12 = StaticBool() {value=true}
  n13 = MaterialFunctionCall(
   Texture Coordinate 0-1=NamedRerouteUsage()
   Top Left Position=MaterialFunctionCall(0-1 Position=Constant2Vector() {g=0.3, r=0.075}, "(XY, YX) Ratios"=@n11) {materialFunction=&MF_Display_PositionFromAspectRatio}.Scaled Position
   Bottom Right Position=MaterialFunctionCall(0-1 Position=Constant2Vector() {g=0.45, r=0.275}, "(XY, YX) Ratios"=@n11) {materialFunction=&MF_Display_PositionFromAspectRatio}.Scaled Position
   Outline Width=Constant() {r=0.05}
   Outline Color=Multiply(A=NamedRerouteUsage()) {constB=0.2}
   Background Color=NamedRerouteUsage()
   Corner Width=Constant() {r=0.5}
   Top Left Corner=@n12
   Bottom Right Corner=@n12
  ) {materialFunction=&MF_GrillMoon_Display_Box}
  n14 = MaterialFunctionCall!(
   Base Element Color=@n10.Combined Color
   Base Element Mask=@n10.Combined Color
   New Element Color=@n13.Box Color
   New Element Mask=@n13.Box Color
  ) {materialFunction=/GrillMoon_Countdown2/Assets/TechArt/CountdownDisplays/Materials/MF_GrillMoon_DisplayElementMixer}
  n15 = NamedRerouteUsage()
  n16 = MaterialFunctionCall!(
   Texture Coordinate 0-1=NamedRerouteUsage()
   Graph Location=MaterialFunctionCall(0-1 Position=Constant2Vector() {g=0.7, r=0.17}, "(XY, YX) Ratios"=NamedRerouteUsage()) {materialFunction=&MF_Display_PositionFromAspectRatio}.Scaled Position
   Position Graph From Corner
   Graph Size=Constant() {r=0.37}
   Time=Time()
   Background Bar Width
   Background Color=NamedRerouteUsage()
   Inner Bar Width
   Inner Bar Trail Width
   Inner Bar Trail Fade
   Swap Trail Direction
   Inner Bar Color=@n15
   "Numbers Value (int)"=Constant() {r=16}
   Numbers Size
   Numbers Color=@n15
  ) {materialFunction=/GrillMoon_Countdown2/Assets/TechArt/CountdownDisplays/Materials/MF_RadialGraph}
  n17 = MaterialFunctionCall(
   Base Element Color=@n14.Combined Color
   Base Element Mask=@n14.Combined Color
   New Element Color=@n16.Graph Color
   New Element Mask=@n16.Graph Color
  ) {materialFunction=&MF_GrillMoon_DisplayElementMixer}
  MaterialFunctionCall(
   Base Element Color=LinearInterpolate(
    A=NamedRerouteUsage()
    B=NamedRerouteUsage()
    Alpha=MaterialFunctionCall!(Y Coord=@n1, Scanline Frequency=@n2, Scanline Scroll Speed, Scanline Max Brightness) {materialFunction=/GrillMoon_Countdown2/Assets/TechArt/CountdownDisplays/Materials/MF_DisplayBackground}.Background Grayscale
   )
   New Element Color=@n17.Combined Color
   New Element Mask=@n17.Combined Color
  ) {materialFunction=&MF_GrillMoon_DisplayElementMixer}
  NamedRerouteDeclaration(None=ComponentMask(None=VectorParameter() {parameterName=UI Secondary Color, defaultValue=<rgba 0.0003 0.013 0.0252 1>}.RGB) {b=true, g=true, r=true}) {name=UI Secondary Color, nodeColor=<rgba 0 1 0.1562 1>}
  NamedRerouteDeclaration(None=ComponentMask(None=VectorParameter() {parameterName=UI Primary Color, defaultValue=<rgba 0.3467 1 0.9823 1>}.RGB) {b=true, g=true, r=true}) {name=UI Primary Color, nodeColor=<rgba 0 1 0.1562 1>}
  NamedRerouteDeclaration(None=ComponentMask(None=VectorParameter() {parameterName=UI Tertiary Color, defaultValue=<rgba 0 0.7347 1 1>}.RGB) {b=true, g=true, r=true}) {name=UI Tertiary Color, nodeColor=<rgba 0 1 0.1562 1>}
  NamedRerouteDeclaration(None=MaterialFunctionCall(0-1 Position=TextureCoordinate(), "(XY, YX) Ratios"=NamedRerouteUsage()) {materialFunction=&MF_Display_PositionFromAspectRatio}.Scaled Position) {name=Aspect Correct UV, nodeColor=<rgba 1 0 0.2297 1>}
  n18 = ScalarParameter() {defaultValue=1.7778, parameterName=Screen Aspect Ratio}
  NamedRerouteDeclaration(None=AppendVector(A=@n18, B=Divide(B=@n18))) {name="(XY, YX) Ratios", nodeColor=<rgba 0 0.5781 1 1>}
 ]
}
