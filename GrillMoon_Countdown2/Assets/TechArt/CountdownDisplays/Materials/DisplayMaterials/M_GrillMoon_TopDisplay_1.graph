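# Material for the first GrillMoon countdown top display, a procedurally drawn UI screen with three tintable UI colors.
# Five vertical bars with fixed, hard-stepped fills and two stacked panels with corners on one diagonal are built from shared box functions.
# Element mixer functions layer these shapes over a scanline background that blends between UI colors.
# A screen aspect ratio parameter drives UV and position correction.
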Material /GrillMoon_Countdown2/Assets/TechArt/CountdownDisplays/Materials/DisplayMaterials/M_GrillMoon_TopDisplay_1
graph {
 properties {bAutomaticallySetUsageInEditor=false, usedShadingModels=Default Lit}
 nodes [
  n1 = ComponentMask(None=NamedRerouteUsage()) {g=true}
  n2 = ScalarParameter() {defaultValue=32, parameterName=Scanline Frequency}
  n3 = NamedRerouteUsage()
  n4 = MaterialFunctionCall!(
   Texture Coordinate 0-1=NamedRerouteUsage()
   Top Left Position=MaterialFunctionCall!(0-1 Position=Constant2Vector() {g=0.15, r=0.4}, "(XY, YX) Ratios"=@n3, Reconstruct YX From XY) {materialFunction=/GrillMoon_Countdown2/Assets/TechArt/CountdownDisplays/Materials/MF_Display_PositionFromAspectRatio}.Scaled Position
   Bottom Right Position=MaterialFunctionCall(0-1 Position=Constant2Vector() {g=0.85, r=0.45}, "(XY, YX) Ratios"=@n3) {materialFunction=&MF_Display_PositionFromAspectRatio}.Scaled Position
   Outline Width=Constant()
   Outline Color=NamedRerouteUsage()
   Background Color=NamedRerouteUsage()
   Background Opacity
   Corner Width
   Top Left Corner
   Top Right Corner
   Bottom Left Corner
   Bottom Right Corner
  ) {materialFunction=/GrillMoon_Countdown2/Assets/TechArt/CountdownDisplays/Materials/MF_GrillMoon_Display_Box}
  n5 = Constant() {r=0.6}
  n6 = ComponentMask(None=@n4.Box Fit UV) {g=true}
  n7 = Ceil(None=Saturate(None=InverseLinearInterpolate(A=@n5, Value=@n6)))
  n8 = MaterialFunctionCall!(
   Box Element Color=@n4.Box Color
   Box Element Mask
   Box Inner Mask=@n4.Box Color
   Whole Box Mask=@n4.Box Color
   New Element Color=Multiply(A=@n7, B=NamedRerouteUsage())
   New Element Mask=@n7
   Merge Combined Mask in to Whole Box Mask
  ) {materialFunction=/GrillMoon_Countdown2/Assets/TechArt/CountdownDisplays/Materials/MF_GrillMoon_DisplayBoxElementMixer}
  n9 = NamedRerouteUsage()
  n10 = MaterialFunctionCall(
   Texture Coordinate 0-1=NamedRerouteUsage()
   Top Left Position=MaterialFunctionCall(0-1 Position=Constant2Vector() {g=0.15, r=0.5}, "(XY, YX) Ratios"=@n9) {materialFunction=&MF_Display_PositionFromAspectRatio}.Scaled Position
   Bottom Right Position=MaterialFunctionCall(0-1 Position=Constant2Vector() {g=0.85, r=0.55}, "(XY, YX) Ratios"=@n9) {materialFunction=&MF_Display_PositionFromAspectRatio}.Scaled Position
   Outline Width=Constant()
   Outline Color=NamedRerouteUsage()
   Background Color=NamedRerouteUsage()
  ) {materialFunction=&MF_GrillMoon_Display_Box}
  n11 = Constant() {r=0.9}
  n12 = ComponentMask(None=@n10.Box Fit UV) {g=true}
  n13 = Ceil(None=Saturate(None=InverseLinearInterpolate(A=@n11, Value=@n12)))
  n14 = MaterialFunctionCall(
   Box Element Color=@n10.Box Color
   Box Inner Mask=@n10.Box Color
   Whole Box Mask=@n10.Box Color
   New Element Color=Multiply(A=@n13, B=NamedRerouteUsage())
   New Element Mask=@n13
  ) {materialFunction=&MF_GrillMoon_DisplayBoxElementMixer}
  n15 = MaterialFunctionCall!(
   Base Element Color=@n8.Combined Color
   Base Element Mask=@n8.Combined Color
   New Element Color=@n14.Combined Color
   New Element Mask=@n14.Combined Color
  ) {materialFunction=/GrillMoon_Countdown2/Assets/TechArt/CountdownDisplays/Materials/MF_GrillMoon_DisplayElementMixer}
  n16 = NamedRerouteUsage()
  n17 = MaterialFunctionCall(
   Texture Coordinate 0-1=NamedRerouteUsage()
   Top Left Position=MaterialFunctionCall(0-1 Position=Constant2Vector() {g=0.15, r=0.6}, "(XY, YX) Ratios"=@n16) {materialFunction=&MF_Display_PositionFromAspectRatio}.Scaled Position
   Bottom Right Position=MaterialFunctionCall(0-1 Position=Constant2Vector() {g=0.85, r=0.65}, "(XY, YX) Ratios"=@n16) {materialFunction=&MF_Display_PositionFromAspectRatio}.Scaled Position
   Outline Width=Constant()
   Outline Color=NamedRerouteUsage()
   Background Color=NamedRerouteUsage()
  ) {materialFunction=&MF_GrillMoon_Display_Box}
  n18 = Constant() {r=0.4}
  n19 = ComponentMask(None=@n17.Box Fit UV) {g=true}
  n20 = Ceil(None=Saturate(None=InverseLinearInterpolate(A=@n18, Value=@n19)))
  n21 = MaterialFunctionCall(
   Box Element Color=@n17.Box Color
   Box Inner Mask=@n17.Box Color
   Whole Box Mask=@n17.Box Color
   New Element Color=Multiply(A=@n20, B=NamedRerouteUsage())
   New Element Mask=@n20
  ) {materialFunction=&MF_GrillMoon_DisplayBoxElementMixer}
  n22 = MaterialFunctionCall(
   Base Element Color=@n15.Combined Color
   Base Element Mask=@n15.Combined Color
   New Element Color=@n21.Combined Color
   New Element Mask=@n21.Combined Color
  ) {materialFunction=&MF_GrillMoon_DisplayElementMixer}
  n23 = NamedRerouteUsage()
  n24 = MaterialFunctionCall(
   Texture Coordinate 0-1=NamedRerouteUsage()
   Top Left Position=MaterialFunctionCall(0-1 Position=Constant2Vector() {g=0.15, r=0.7}, "(XY, YX) Ratios"=@n23) {materialFunction=&MF_Display_PositionFromAspectRatio}.Scaled Position
   Bottom Right Position=MaterialFunctionCall(0-1 Position=Constant2Vector() {g=0.85, r=0.75}, "(XY, YX) Ratios"=@n23) {materialFunction=&MF_Display_PositionFromAspectRatio}.Scaled Position
   Outline Width=Constant()
   Outline Color=NamedRerouteUsage()
   Background Color=NamedRerouteUsage()
  ) {materialFunction=&MF_GrillMoon_Display_Box}
  n25 = Constant() {r=0.3}
  n26 = ComponentMask(None=@n24.Box Fit UV) {g=true}
  n27 = Ceil(None=Saturate(None=InverseLinearInterpolate(A=@n25, Value=@n26)))
  n28 = MaterialFunctionCall(
   Box Element Color=@n24.Box Color
   Box Inner Mask=@n24.Box Color
   Whole Box Mask=@n24.Box Color
   New Element Color=Multiply(A=@n27, B=NamedRerouteUsage())
   New Element Mask=@n27
  ) {materialFunction=&MF_GrillMoon_DisplayBoxElementMixer}
  n29 = MaterialFunctionCall(
   Base Element Color=@n22.Combined Color
   Base Element Mask=@n22.Combined Color
   New Element Color=@n28.Combined Color
   New Element Mask=@n28.Combined Color
  ) {materialFunction=&MF_GrillMoon_DisplayElementMixer}
  n30 = NamedRerouteUsage()
  n31 = MaterialFunctionCall(
   Texture Coordinate 0-1=NamedRerouteUsage()
   Top Left Position=MaterialFunctionCall(0-1 Position=Constant2Vector() {g=0.15, r=0.8}, "(XY, YX) Ratios"=@n30) {materialFunction=&MF_Display_PositionFromAspectRatio}.Scaled Position
   Bottom Right Position=MaterialFunctionCall(0-1 Position=Constant2Vector() {g=0.85, r=0.85}, "(XY, YX) Ratios"=@n30) {materialFunction=&MF_Display_PositionFromAspectRatio}.Scaled Position
   Outline Width=Constant()
   Outline Color=NamedRerouteUsage()
   Background Color=NamedRerouteUsage()
  ) {materialFunction=&MF_GrillMoon_Display_Box}
  n32 = Constant() {r=0.7}
  n33 = ComponentMask(None=@n31.Box Fit UV) {g=true}
  n34 = Ceil(None=Saturate(None=InverseLinearInterpolate(A=@n32, Value=@n33)))
  n35 = MaterialFunctionCall(
   Box Element Color=@n31.Box Color
   Box Inner Mask=@n31.Box Color
   Whole Box Mask=@n31.Box Color
   New Element Color=Multiply(A=@n34, B=NamedRerouteUsage())
   New Element Mask=@n34
  ) {materialFunction=&MF_GrillMoon_DisplayBoxElementMixer}
  n36 = MaterialFunctionCall(
   Base Element Color=@n29.Combined Color
   Base Element Mask=@n29.Combined Color
   New Element Color=@n35.Combined Color
   New Element Mask=@n35.Combined Color
  ) {materialFunction=&MF_GrillMoon_DisplayElementMixer}
  n37 = NamedRerouteUsage()
  n38 = StaticBool() {value=true}
  n39 = MaterialFunctionCall(
   Texture Coordinate 0-1=NamedRerouteUsage()
   Top Left Position=MaterialFunctionCall(0-1 Position=Constant2Vector() {g=0.2, r=0.075}, "(XY, YX) Ratios"=@n37) {materialFunction=&MF_Display_PositionFromAspectRatio}.Scaled Position
   Bottom Right Position=MaterialFunctionCall(0-1 Position=Constant2Vector() {g=0.35, r=0.35}, "(XY, YX) Ratios"=@n37) {materialFunction=&MF_Display_PositionFromAspectRatio}.Scaled Position
   Outline Width=Constant() {r=0.05}
   Outline Color=Multiply(A=NamedRerouteUsage()) {constB=0.2}
   Background Color=NamedRerouteUsage()
   Corner Width=Constant() {r=0.5}
   Top Left Corner=@n38
   Bottom Right Corner=@n38
  ) {materialFunction=&MF_GrillMoon_Display_Box}
  n40 = NamedRerouteUsage()
  n41 = StaticBool() {value=true}
  n42 = MaterialFunctionCall(
   Texture Coordinate 0-1=NamedRerouteUsage()
   Top Left Position=MaterialFunctionCall(0-1 Position=Constant2Vector() {g=0.4, r=0.075}, "(XY, YX) Ratios"=@n40) {materialFunction=&MF_Display_PositionFromAspectRatio}.Scaled Position
   Bottom Right Position=MaterialFunctionCall(0-1 Position=Constant2Vector() {g=0.55, r=0.35}, "(XY, YX) Ratios"=@n40) {materialFunction=&MF_Display_PositionFromAspectRatio}.Scaled Position
   Outline Width=Constant() {r=0.05}
   Outline Color=Multiply(A=NamedRerouteUsage()) {constB=0.2}
   Background Color=Multiply(A=NamedRerouteUsage()) {constB=0.1}
   Corner Width=Constant() {r=0.3}
   Top Left Corner=@n41
   Bottom Right Corner=@n41
  ) {materialFunction=&MF_GrillMoon_Display_Box}
  n43 = MaterialFunctionCall(Base Element Color=@n39.Box Color, Base Element Mask=@n39.Box Color, New Element Color=@n42.Box Color, New Element Mask=@n42.Box Color) {materialFunction=&MF_GrillMoon_DisplayElementMixer}
  n44 = MaterialFunctionCall(
   Base Element Color=@n36.Combined Color
   Base Element Mask=@n36.Combined Color
   New Element Color=@n43.Combined Color
   New Element Mask=@n43.Combined Color
  ) {materialFunction=&MF_GrillMoon_DisplayElementMixer}
  MaterialFunctionCall(
   Base Element Color=LinearInterpolate(
    A=NamedRerouteUsage()
    B=NamedRerouteUsage()
    Alpha=MaterialFunctionCall!(Y Coord=@n1, Scanline Frequency=@n2, Scanline Scroll Speed, Scanline Max Brightness) {materialFunction=/GrillMoon_Countdown2/Assets/TechArt/CountdownDisplays/Materials/MF_DisplayBackground}.Background Grayscale
   )
   New Element Color=@n44.Combined Color
   New Element Mask=@n44.Combined Color
  ) {materialFunction=&MF_GrillMoon_DisplayElementMixer}
  NamedRerouteDeclaration(None=ComponentMask(None=VectorParameter() {parameterName=UI Secondary Color, defaultValue=<rgba 0.0003 0.013 0.0252 1>}.RGB) {b=true, g=true, r=true}) {name=UI Secondary Color, nodeColor=<rgba 0 1 0.1562 1>}
  NamedRerouteDeclaration(None=ComponentMask(None=VectorParameter() {parameterName=UI Primary Color, defaultValue=<rgba 0.3467 1 0.9823 1>}.RGB) {b=true, g=true, r=true}) {name=UI Primary Color, nodeColor=<rgba 0 1 0.1562 1>}
  NamedRerouteDeclaration(None=ComponentMask(None=VectorParameter() {parameterName=UI Tertiary Color, defaultValue=<rgba 0 0.7347 1 1>}.RGB) {b=true, g=true, r=true}) {name=UI Tertiary Color, nodeColor=<rgba 0 1 0.1562 1>}
  NamedRerouteDeclaration(None=MaterialFunctionCall(0-1 Position=TextureCoordinate(), "(XY, YX) Ratios"=NamedRerouteUsage()) {materialFunction=&MF_Display_PositionFromAspectRatio}.Scaled Position) {name=Aspect Correct UV, nodeColor=<rgba 1 0 0.2297 1>}
  n45 = ScalarParameter() {defaultValue=1.7778, parameterName=Screen Aspect Ratio}
  NamedRerouteDeclaration(None=AppendVector(A=@n45, B=Divide(B=@n45))) {name="(XY, YX) Ratios", nodeColor=<rgba 0 0.5781 1 1>}
 ]
}
